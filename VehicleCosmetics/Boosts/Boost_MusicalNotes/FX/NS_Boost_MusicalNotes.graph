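# Niagara system for the Musical Notes vehicle boost cosmetic, trailing note sprites and ribbons from the vehicle's skeletal mesh sockets.
# Each emitter filters socket count to distribute particles per port, firing them backward with inherited source movement.
# Note sprites use four tagged spawn groups and four renderers with separate SubUV frame ranges, tinted by random color.
# Two thin cyan ribbons layer with a wide, width-pulsing trail; painted-variant overrides are disabled.
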NiagaraSystem /VehicleCosmetics/Boosts/Boost_MusicalNotes/FX/NS_Boost_MusicalNotes
dependencies {
 usedDataInterfaces [
  &NiagaraDataInterfaceCurve
  &NiagaraDataInterfaceVectorCurve
  &NiagaraDataInterfaceSpriteRendererInfo
  &NiagaraDataInterfaceSkeletalMesh
 ]
 usedDynamicInputs []
 usedModules [
  &self:SystemSpawnScript.SystemScriptSource.SystemScriptGraph.NiagaraNodeAssignment_0.SetVariables_#1
  &SystemState
 ]
}
emitters {
 MusicalNotes001 {
  emitterUpdateScript [
   &EmitterState()
   &VC_FilteredSocketCount()
   &SpawnRate(SpawnRate=&RandomRangeFloat(Minimum=20, Maximum=15), Use Spawn Probability=false)
   &SpawnRate(SpawnGroup=1, SpawnRate=&RandomRangeFloat(Minimum=20, Maximum=15), Use Spawn Probability=false) {module=SpawnRate001}
   &SpawnRate(SpawnGroup=2, SpawnRate=&RandomRangeFloat(Minimum=20, Maximum=15), Use Spawn Probability=false) {module=SpawnRate002}
   &SpawnRate(SpawnGroup=3, SpawnRate=&RandomRangeFloat(Minimum=20, Maximum=15), Use Spawn Probability=false) {module=SpawnRate003}
  ]
  particleSpawnScript [
   &VC_PaintedVariantsOverride() {enabled=false}
   /SetVariables(Particles.Tag_Notes:NiagaraInt32=&SpawnGroup()) {isSetParameters=true}
   &InitializeParticle(
    Color Mode=::NewEnumerator0
    Lifetime=&RandomRangeFloat(Minimum=1, Maximum=0.6)
    Sprite Rotation Angle=&RandomRangeFloat(Minimum=-0.1, Maximum=0.1)
    Sprite Rotation Mode=::NewEnumerator1
    Sprite Size Mode=::NewEnumerator1
    Uniform Sprite Size Max=&Multiply_Float(A=3, B=3)
    Uniform Sprite Size Min=&Multiply_Float(A=8, B=3)
    UsePositionOffset=false
    Write Material Random=true
   )
   &PartitionParticles(Number Of Partitions=$Emitter.PortCount)
   &SkeletalMeshLocation(
    Bone / Socket Index=$Particles.PartitionParticles.Partition
    Bone / Socket Sampling Mode=::NewEnumerator5
    MaskSpawnGroup=false
    Mesh Sampling Type=::NewEnumerator6
    Sampled Position Offset=<-20 0 0>
    Skeletal Mesh=$System.SkeletalMesh
   )
   &AddVelocity(Velocity=&RandomRangeVector(Minimum=<-1200 75 90>, Maximum=<-800 -75 55>))
   &Color(Color=&RandomRangeLinearColor(Minimum=<rgba 1 0.01 0.02 1>, Maximum=<rgba 0.01 1 0.2 1>))
   &ShapeLocation(
    MaskSpawnGroup=false
    Override Seed=false
    Sphere Radius=&RandomRangeFloat(Minimum=20, Maximum=40)
    UseOwnerScale=true
   )
  ]
  particleUpdateScript [
   &ParticleState()
   &ScaleSpriteSize(
    Uniform Curve Scale=3
    Uniform Curve Sprite Scale=DataInterface() {propertyValues {Curve {keys [{tangentMode=RCTM_User, time=0, value=1}, {tangentMode=RCTM_User, time=1, value=1}]}}}
   )
   &InheritSourceMovement(
    Applied Position Scale=&Multiply_VectorByFloat(
     Vector=<1 1 1>
     Float=&FloatFromCurve(FloatCurve=DataInterface() {propertyValues {Curve {keys [
          {interpMode=RCIM_Cubic, leaveTangent=-0.0051, tangentMode=RCTM_Break, time=0, value=1}
          {arriveTangent=-0.0476, interpMode=RCIM_Cubic, tangentMode=RCTM_Break, time=1, value=0.9}
         ]}}})
    )
    ApplyForce=false
    ApplyPosition=true
    ApplyPositionFalloff=false
   )
   &SolveForcesAndVelocity(Clamp Velocity=false, Limit Acceleration=false)
   &ScaleColor(Scale Alpha=5, ScaleA=true, ScaleRGB=true)
   &DynamicMaterialParameters(
    Index 0 Param 1=1
    Module.Index0Float0Write=true
    Module.Index0Float1Write=true
    Module.Index0Float2Write=true
    Module.Index0Float3Write=true
   )
   &SubUVAnimation(
    Output Alternate Renderer Binding=true
    Play Rate:NiagaraFloat=&RandomRangeFloat(Minimum=2, Maximum=4)
    Playback Scale Curve=DataInterface() {propertyValues {Curve {keys [
        {arriveTangent=16.7186, interpMode=RCIM_Cubic, leaveTangent=16.7186, tangentMode=RCTM_User, time=0, value=49}
        {arriveTangent=15.3492, interpMode=RCIM_Cubic, leaveTangent=15.3492, tangentMode=RCTM_User, time=1, value=63}
       ]}}}
    "Sprite Renderer "=DataInterface() {propertyValues {SpriteRenderer=&self:MusicalNotes001_0.NiagaraSpriteRendererProperties_3}}
    Start Frame Range Override=49
    SubUV Animation Mode:ENiagara_SubUVLookupModeV2=::NewEnumerator3
    UseEndFrame=true
    UsePlayRateScale=false
    UseStartFrame=true
   )
   &SubUVAnimation(
    End Frame Range Override=48
    Output Alternate Renderer Binding=true
    Play Rate:NiagaraFloat=&RandomRangeFloat(Minimum=2, Maximum=4)
    Playback Scale Curve=DataInterface() {propertyValues {Curve {keys [
        {arriveTangent=17.0031, interpMode=RCIM_Cubic, leaveTangent=17.0031, tangentMode=RCTM_User, time=0, value=32}
        {arriveTangent=17.3201, interpMode=RCIM_Cubic, leaveTangent=17.3201, tangentMode=RCTM_User, time=1, value=48}
       ]}}}
    "Sprite Renderer "=DataInterface() {propertyValues {SpriteRenderer=&self:MusicalNotes001_0.NiagaraSpriteRendererProperties_2}}
    Start Frame Range Override=32
    SubUV Animation Mode:ENiagara_SubUVLookupModeV2=::NewEnumerator3
    UseEndFrame=true
    UsePlayRateScale=false
    UseStartFrame=true
   ) {module=SubUVAnimation001}
   &SubUVAnimation(
    End Frame Range Override=31
    Output Alternate Renderer Binding=true
    Play Rate:NiagaraFloat=&RandomRangeFloat(Minimum=2, Maximum=4)
    Playback Scale Curve=DataInterface() {propertyValues {Curve {keys [
        {arriveTangent=14.2868, interpMode=RCIM_Cubic, leaveTangent=14.2868, tangentMode=RCTM_User, time=0, value=16}
        {arriveTangent=14.9086, interpMode=RCIM_Cubic, leaveTangent=14.9086, tangentMode=RCTM_User, time=1, value=31}
       ]}}}
    "Sprite Renderer "=DataInterface() {propertyValues {SpriteRenderer=&self:MusicalNotes001_0.NiagaraSpriteRendererProperties_4}}
    Start Frame Range Override=16
    SubUV Animation Mode:ENiagara_SubUVLookupModeV2=::NewEnumerator3
    UseEndFrame=true
    UsePlayRateScale=false
    UseStartFrame=true
   ) {module=SubUVAnimation002}
   &SubUVAnimation(
    End Frame Range Override=15
    Output Alternate Renderer Binding=true
    Play Rate:NiagaraFloat=&RandomRangeFloat(Minimum=2, Maximum=4)
    Playback Scale Curve=DataInterface() {propertyValues {Curve {keys [
        {arriveTangent=13.356, interpMode=RCIM_Cubic, leaveTangent=13.356, tangentMode=RCTM_User, time=0, value=0}
        {arriveTangent=14.4867, interpMode=RCIM_Cubic, leaveTangent=14.4867, tangentMode=RCTM_User, time=1, value=15}
       ]}}}
    "Sprite Renderer "=DataInterface() {propertyValues {SpriteRenderer=&self:MusicalNotes001_0.NiagaraSpriteRendererProperties_5}}
    SubUV Animation Mode:ENiagara_SubUVLookupModeV2=::NewEnumerator3
    UseEndFrame=true
    UsePlayRateScale=false
    UseStartFrame=true
   ) {module=SubUVAnimation003}
  ]
  rendererClasses [&NiagaraSpriteRendererProperties]
  renderers [
   {rendererClass=&NiagaraSpriteRendererProperties}
   {rendererClass=&NiagaraSpriteRendererProperties, rendererIndex=1}
   {rendererClass=&NiagaraSpriteRendererProperties, rendererIndex=2}
   {rendererClass=&NiagaraSpriteRendererProperties, rendererIndex=3}
  ]
 }
 NE_VC_Boost_Ribbon_Parent {
  emitterUpdateScript [
   &EmitterState()
   &VC_FilteredSocketCount()
   &VC_RateQualityHandler(Epic Spawn Rate=80, High Spawn Rate=80, Low Spawn Rate=80, Medium Spawn Rate=80, Use Custom Rate?=true)
   &VC_SpawnRibbon_v1(Number of Ribbons=$Emitter.PortCount, SpawnRate=$Emitter.Spawn Rate)
  ]
  particleSpawnScript [
   &VC_PaintedVariantsOverride() {enabled=false}
   &VC_InitializeRibbon_v1(
    Color=<rgba 0.954 5.3434 6.5708 2.5985>
    Custom Offset=true
    Enable Skeletal Position=true
    Lifetime=1.1
    Local Vector Offset=&AddVectorToPosition(Position=&Multiply_VectorByFloat(Vector=<0 6 -8>, Float=4), Vector=<0 0 0>)
    Offset Amount=1
    Ribbon Width:NiagaraFloat=&Multiply_Float(A=0.5, B=7)
   )
   &AddVelocity(Velocity=<-2500 0 0>) {module=AddVelocity001}
   /SetVariables(
    Particles.Position_Ribbon1:Vector3f=&AddVectorToPosition(Position=<0 0 0>, Vector=<0 0 0>)
    Particles.SecondRibbon:LinearColor=<rgba 1.5 5.5 7 1>
   ) {isSetParameters=true, module=SetVariables_#2}
   &Color(Color=<rgba 1.5 5 8 1>)
  ]
  particleUpdateScript [
   &ParticleState()
   &InheritSourceMovement(
    Applied Position Scale=&Multiply_VectorByFloat(
     Vector=<1 1 1>
     Float=&FloatFromCurve(FloatCurve=DataInterface() {propertyValues {Curve {keys [
          {interpMode=RCIM_Cubic, leaveTangent=-0.0051, tangentMode=RCTM_Break, time=0, value=1}
          {arriveTangent=-0.0476, interpMode=RCIM_Cubic, tangentMode=RCTM_Break, time=1, value=0.9}
         ]}}})
    )
    ApplyForce=false
    ApplyPosition=true
    ApplyPositionFalloff=false
   )
   &SolveForcesAndVelocity(Clamp Velocity=false, Limit Acceleration=false)
   &VC_UpdateRibbon_v1(Enable Skeletal Position=true)
   &ScaleColor(
    Scale Alpha=&FloatFromCurve(FloatCurve=DataInterface() {propertyValues {Curve {keys [
         <at 0 0>
         {interpMode=RCIM_Cubic, leaveTangent=-0.8489, tangentMode=RCTM_Break, time=0.15, value=1}
         {arriveTangent=-0.9483, interpMode=RCIM_Cubic, tangentMode=RCTM_Break, time=1, value=0.2}
        ]}}})
    ScaleA=true
    ScaleRGB=true
   )
   &DynamicMaterialParameters(
    Index 0 Param 1=0.18
    Index 0 Param 2=2.5
    Index 0 Param 3=0.104
    Index 0 Param 4=1.25
    Module.Index0Float0Write=true
    Module.Index0Float1Write=true
    Module.Index0Float2Write=true
    Module.Index0Float3Write=true
   )
   /SetVariables(Particles.DMP_SecondRibbon:Vector4f=&MakeVector4(X=0.45, Y=3, Z=0.2, W=1)) {enabled=false, isSetParameters=true, module=SetVariables_#3}
  ]
  rendererClasses [&NiagaraRibbonRendererProperties]
  renderers [
   {rendererClass=&NiagaraRibbonRendererProperties}
   {rendererClass=&NiagaraRibbonRendererProperties, rendererIndex=1}
  ]
 }
 NE_VC_Boost_Ribbon_Parent001 {
  emitterUpdateScript [
   &EmitterState()
   &VC_FilteredSocketCount()
   &VC_RateQualityHandler(Epic Spawn Rate=120, High Spawn Rate=100, Low Spawn Rate=100, Medium Spawn Rate=100, Use Custom Rate?=true)
   &VC_SpawnRibbon_v1(Number of Ribbons=$Emitter.PortCount, SpawnRate=$Emitter.Spawn Rate)
  ]
  particleSpawnScript [
   &VC_PaintedVariantsOverride() {enabled=false}
   &VC_InitializeRibbon_v1(
    Color=<rgba 1.5 5.686 8.2424 1.3989>
    Custom Offset=true
    Enable Skeletal Position=true
    Lifetime=0.9
    Local Vector Offset=&Multiply_VectorByFloat(Vector=<0 -6 0>, Float=1)
    Offset Amount=1
    Ribbon Width:NiagaraFloat=&Multiply_Float(A=0.5, B=6)
   )
   &AddVelocity(Velocity=<-2500 0 0>) {module=AddVelocity001}
   /SetVariables(
    Particles.Position_Ribbon1:Vector3f=&AddVectorToPosition(Position=<0 0 0>, Vector=<0 0 0>)
    Particles.SecondRibbon:LinearColor=<rgba 0.4298 5.5 7 1>
   ) {enabled=false, isSetParameters=true, module=SetVariables_#2}
   &Color(Color=<rgba 1.5 5.5 7 1>)
  ]
  particleUpdateScript [
   &ParticleState()
   &InheritSourceMovement(
    Applied Position Scale=&Multiply_VectorByFloat(
     Vector=<1 1 1>
     Float=&FloatFromCurve(FloatCurve=DataInterface() {propertyValues {Curve {keys [
          {interpMode=RCIM_Cubic, leaveTangent=-0.0051, tangentMode=RCTM_Break, time=0, value=1}
          {arriveTangent=-0.0476, interpMode=RCIM_Cubic, tangentMode=RCTM_Break, time=1, value=0.9}
         ]}}})
    )
    ApplyForce=false
    ApplyPosition=true
    ApplyPositionFalloff=false
   )
   &SolveForcesAndVelocity(Clamp Velocity=false, Limit Acceleration=false)
   &VC_UpdateRibbon_v1(Enable Skeletal Position=true)
   &ScaleColor(
    Scale Alpha=&FloatFromCurve(FloatCurve=DataInterface() {propertyValues {Curve {keys [<at 0 0>,<at 0.15 1>,<at 1 0.2>]}}})
    ScaleA=true
    ScaleRGB=true
   )
   &DynamicMaterialParameters(
    Index 0 Param 1=0.176
    Index 0 Param 2=1
    Index 0 Param 3=0.2
    Index 0 Param 4=1
    Module.Index0Float0Write=true
    Module.Index0Float1Write=true
    Module.Index0Float2Write=true
    Module.Index0Float3Write=true
   ) {enabled=false}
   /SetVariables(Particles.DMP_SecondRibbon:Vector4f=&MakeVector4(X=1.5, Y=2, Z=0.152, W=0.632)) {isSetParameters=true, module=SetVariables_#3}
  ]
  rendererClasses [&NiagaraRibbonRendererProperties]
  renderers [
   {rendererClass=&NiagaraRibbonRendererProperties}
   {rendererClass=&NiagaraRibbonRendererProperties, rendererIndex=1}
  ]
 }
 Trail_Notes {
  emitterUpdateScript [
   &EmitterState()
   &VC_FilteredSocketCount()
   &VC_RateQualityHandler(High Spawn Rate=60, Low Spawn Rate=45, Medium Spawn Rate=50, Use Custom Rate?=true)
   &VC_SpawnRibbon_v1(Number of Ribbons=$Emitter.PortCount, Spawn Time=-1, SpawnRate=$Emitter.Spawn Rate)
  ]
  particleSpawnScript [
   &VC_PaintedVariantsOverride() {enabled=false}
   &VC_InitializeRibbon_v1(
    Color=<rgba 0.75 3 5 1>
    Custom Offset=true
    Enable Skeletal Position=true
    Facing Axis:ENiagaraOrientationAxis=::ZAxis
    Lifetime=0.9
    Offset Amount=1
    Ribbon Width=85
   )
   &AddVelocity(Velocity=<-2500 0 0>) {module=AddVelocity001}
   &Color(Color=<rgba 1 3 6 1>) {enabled=false}
  ]
  particleUpdateScript [
   &ParticleState()
   &InheritSourceMovement(
    Applied Position Scale=&Multiply_VectorByFloat(
     Vector=<1 1 1>
     Float=&FloatFromCurve(FloatCurve=DataInterface() {propertyValues {Curve {keys [<at 0 1 RCIM_Cubic>,<at 1 0 RCIM_Cubic>]}}})
    )
    ApplyForce=false
    ApplyPosition=true
    ApplyPositionFalloff=false
   ) {enabled=false}
   &InheritSourceMovement(
    Applied Position Scale=&Multiply_VectorByFloat(
     Vector=<1 1 1>
     Float=&FloatFromCurve(FloatCurve=DataInterface() {propertyValues {Curve {keys [
          {interpMode=RCIM_Cubic, leaveTangent=-0.0051, tangentMode=RCTM_Break, time=0, value=1}
          {arriveTangent=-0.0476, interpMode=RCIM_Cubic, tangentMode=RCTM_Break, time=1, value=0.9}
         ]}}})
    )
    ApplyForce=false
    ApplyPosition=true
    ApplyPositionFalloff=false
   ) {module=InheritSourceMovement001}
   &SolveForcesAndVelocity(Clamp Velocity=false, Limit Acceleration=false)
   &VC_UpdateRibbon_v1(Enable Skeletal Position=true)
   &ScaleRibbonWidth(Ribbon Width Scale=&MakeFloatFromVector(
     Vector=&VectorFromCurve(VectorCurve=DataInterface() {propertyValues {
        XCurve {keys [
          {interpMode=RCIM_Cubic, tangentMode=RCTM_Break, time=0, value=1}
          {arriveTangent=-9.587, interpMode=RCIM_Cubic, leaveTangent=5.6116, tangentMode=RCTM_Break, time=0.25, value=0}
          {arriveTangent=4, interpMode=RCIM_Cubic, leaveTangent=4, time=0.3, value=0.3}
          {interpMode=RCIM_Cubic, tangentMode=RCTM_Break, time=0.5, value=1}
          {arriveTangent=-5.4239, interpMode=RCIM_Cubic, leaveTangent=3.063, tangentMode=RCTM_Break, time=0.75, value=0}
          {interpMode=RCIM_Cubic, tangentMode=RCTM_Break, time=1, value=1}
         ]}
        YCurve {keys [<at 0 0>,<at 1 0>]}
        ZCurve {keys [<at 0 0>,<at 1 0>]}
       }})
     Channel=::NewEnumerator0
    )) {enabled=false}
   &ScaleRibbonWidth(Ribbon Width Scale=&MakeFloatFromVector(
     Vector=&VectorFromCurve(VectorCurve=DataInterface() {propertyValues {
        XCurve {
         defaultValue=1
         keys [
          <at 0 1>
          {interpMode=RCIM_Cubic, tangentMode=RCTM_Break, time=0.3, value=1}
          {interpMode=RCIM_Cubic, tangentMode=RCTM_Break, time=0.5, value=0.4}
          <at 0.7 0.85>
          {interpMode=RCIM_Cubic, tangentMode=RCTM_Break, time=1, value=0}
         ]
        }
        YCurve {keys [<at 0 0>,<at 1 0>]}
        ZCurve {defaultValue=1, keys [<at 0 0>,<at 1 0>]}
       }})
     Channel=::NewEnumerator0
    )) {enabled=false, module=ScaleRibbonWidth001}
   &ScaleRibbonWidth(Ribbon Width Scale=&MakeFloatFromVector(
     Vector=&VectorFromCurve(VectorCurve=DataInterface() {propertyValues {
        XCurve {keys [
          {interpMode=RCIM_Cubic, leaveTangent=-0.8193, tangentMode=RCTM_Break, time=0, value=1}
          {arriveTangent=-6.576, interpMode=RCIM_Cubic, leaveTangent=5.5247, tangentMode=RCTM_Break, time=0.4, value=-0.025}
          {arriveTangent=0.6816, interpMode=RCIM_Cubic, tangentMode=RCTM_Break, time=0.7, value=0.975}
          {arriveTangent=-5.2251, interpMode=RCIM_Cubic, leaveTangent=3.063, tangentMode=RCTM_Break, time=0.9, value=-0.025}
          {interpMode=RCIM_Cubic, tangentMode=RCTM_Break, time=1, value=0.5}
         ]}
        YCurve {keys [<at 0 0>,<at 1 0>]}
        ZCurve {keys [<at 0 0>,<at 1 0>]}
       }})
     Channel=::NewEnumerator0
    )) {module=ScaleRibbonWidth002}
   &ScaleColor(
    Scale Alpha=&FloatFromCurve(FloatCurve=DataInterface() {propertyValues {Curve {keys [<at 0 0>,<at 0.15 1>,<at 1 1>]}}}, Scale Curve=2)
    ScaleA=true
    ScaleRGB=true
   )
   &DynamicMaterialParameters(
    Index 0 Param 1=&RandomRangeFloat(Minimum=0, Maximum=1)
    Index 0 Param 2=1
    Index 0 Param 4=&FloatFromCurve(FloatCurve=DataInterface() {propertyValues {Curve {keys [<at 0 1>,<at 0.5 1>,<at 1 0>]}}})
    Module.Param0WriteEnabled=true
    Module.Param1WriteEnabled=true
    Module.Param2WriteEnabled=true
    Module.Param3WriteEnabled=true
   )
  ]
  rendererClasses [&NiagaraRibbonRendererProperties]
  renderers [{rendererClass=&NiagaraRibbonRendererProperties}]
 }
}
summary {
 systemName=NS_Boost_MusicalNotes
 emitters [
  {emitterName=Trail_Notes, rendererClasses [&NiagaraRibbonRendererProperties]}
  {emitterName=NE_VC_Boost_Ribbon_Parent, rendererClasses [&NiagaraRibbonRendererProperties]}
  {emitterName=NE_VC_Boost_Ribbon_Parent001, rendererClasses [&NiagaraRibbonRendererProperties]}
  {emitterName=MusicalNotes001, rendererClasses [&NiagaraSpriteRendererProperties]}
 ]
 userVariables [
  User.bIsFrontendPreview {type=NiagaraBool}
  User.bIsPainted {type=NiagaraBool}
  User.Color Array {
   type=NiagaraDataInterfaceArrayColor
   description=""
   defaultValue {
    struct=NiagaraExt_VariableValue_DataInterface
    value {
     dataInterface=&self:NiagaraDataInterfaceArrayColor_0
     dataInterfaceClass=&NiagaraDataInterfaceArrayColor
    }
   }
  }
  User.ColorVariantID {type=NiagaraInt32}
  User.DTCarOffset {type=Vector3f, description="", defaultValue {struct=Vector3f, value=<0 0 0>}}
  User.Linear Color {type=LinearColor, description="", defaultValue {struct=LinearColor, value=<rgba 1 1 1 1>}}
  User.PaintedColor {type=LinearColor, description="", defaultValue {struct=LinearColor, value=<rgba 1 1 1 1>}}
  User.PaintedScalar {type=NiagaraFloat, defaultValue=1, description=""}
  User.PortCount {type=NiagaraInt32, defaultValue=1}
 ]
}
