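# Translucent unlit Niagara ribbon material for the music sheet in the Musical Notes vehicle boost.
# A rainbow texture scaled by dynamic-parameter brightness feeds an emissive adjustment function, while sheet textures, panned or static, are masked by a U-axis fade and dynamic or particle opacity.
# Summed sine and cosine waves over time and UV build world-position-offset style wobble, in two variants: constant-driven, or using particle color channels as frequency, speed and amplitude.
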Material /VehicleCosmetics/Boosts/Boost_MusicalNotes/Material/M_MusicalNotes_MusicSheet
graph {
 properties {bUsedWithNiagaraRibbons=true, blendMode=BLEND_Translucent, pixelDepthOffsetMode=PDOM_Legacy, shadingModel=MSM_Unlit, usedShadingModels=Unlit}
 nodes [
  n1 = DynamicParameter() {paramNames [color, Color_Brightness, Desaturate, Opacity]}
  MaterialFunctionCall(Emissive=Multiply(A=TextureSample() {texture=/DealershipSharedLibrary/FX/Textures/Diffuse/T_VCVFX_Rainbow_D}.RGB, B=@n1.Color_Brightness)) {materialFunction=&MF_FX_EmissiveAdjustments}
  n2 = Panner(Coordinate=TextureCoordinate() {coordinateIndex=1, uTiling=10}) {speedX=-8}
  n3 = TextureCoordinate()
  MaterialExpressionConstantClamp_0 = Saturate(None=Multiply(A=ComponentMask(None=@n3) {r=true}, B=Constant() {r=8}))
  Multiply(
   A=@n1.Opacity
   B=Multiply(
    A=TextureSampleParameter2D(UVs=@n2) {parameterName=Texture, texture=/VehicleCosmetics/Boosts/Boost_MusicalNotes/Textures/T_VCVFX_Boost_MusicNotes_MusicSheet_A}.RGB
    B=@MaterialExpressionConstantClamp_0
   )
  )
  n4 = TextureCoordinate() {coordinateIndex=1, uTiling=10}
  n5 = Multiply(A=ComponentMask(None=TextureCoordinate()) {r=true}, B=Constant() {r=8})
  Multiply(
   A=ParticleColor().A
   B=Multiply(
    A=TextureSampleParameter2D(UVs=@n4) {parameterName=Texture, texture=&T_VCVFX_Boost_MusicNotes_MusicSheet_A}.RGB
    B=Saturate(None=@n5)
   )
  )
  Multiply(
   A=TextureSample() {texture=&T_VCVFX_Rainbow_D}.RGB
   B=DynamicParameter() {paramNames [color, Color_Brightness, Desaturate, Opacity]}.Color_Brightness
  )
  n6 = Constant3Vector() {constant=<rgba 1 3 6 0>}
  n7 = Constant() {r=1}
  n8 = Time()
  n9 = Multiply(A=@n6.G, B=OneMinus(None=Multiply(A=@n7, B=@n8)))
  n10 = ComponentMask(None=Multiply(A=@n6.R, B=TextureCoordinate() {coordinateIndex=1, uTiling=3})) {r=true}
  n11 = Add(A=@n9, B=@n10)
  n12 = Multiply(A=@n9, B=Constant() {r=0.5})
  n13 = Add(A=Add(A=Cosine(None=@n11), B=Cosine(None=@n11) {period=2}), B=Cosine(None=Add(A=@n10, B=@n12)) {period=0.75})
  n14 = Add(A=@n10, B=@n9)
  n15 = Add(A=Add(A=Sine(None=@n14), B=Sine(None=@n14) {period=2}), B=Sine(None=@n14) {period=0.75})
  n16 = Multiply(A=AppendVector(A=AppendVector(A=@n13, B=@n15), B=@n13), B=@n6.B)
  n17 = TextureCoordinate()
  n18 = Saturate(None=Multiply(A=ComponentMask(None=@n17) {r=true}, B=Constant() {r=2}))
  Multiply(A=Transform(None=Multiply(A=@n16, B=@n18)), B=ScalarParameter() {defaultValue=1, parameterName=Multiplier_WPO}) {constB=15}
  MaterialExpressionVertexColor_1 = ParticleColor()
  n19 = Multiply(A=@MaterialExpressionVertexColor_1.G, B=OneMinus(None=Time()))
  n20 = ComponentMask(None=Multiply(A=TextureCoordinate() {coordinateIndex=1, uTiling=3}, B=@MaterialExpressionVertexColor_1.R)) {r=true}
  n21 = Add(A=@n19, B=@n20)
  n22 = Multiply(A=@n19, B=Constant() {r=0.5})
  n23 = Add(A=Add(A=Cosine(None=@n21), B=Cosine(None=@n21) {period=2}), B=Cosine(None=Add(A=@n20, B=@n22)) {period=0.75})
  n24 = Add(A=@n20, B=@n19)
  n25 = Sine(None=@n24)
  n26 = Sine(None=@n24) {period=2}
  n27 = AppendVector(A=@n23, B=Add(A=Add(A=@n25, B=@n26), B=Sine(None=@n24) {period=0.75}))
  n28 = Multiply(A=Multiply(A=AppendVector(A=@n27, B=@n23), B=@MaterialExpressionVertexColor_1.B), B=@MaterialExpressionVertexColor_1.A)
  Reroute(None=Multiply(
    A=Transform(None=@n28) {transformSourceType=TRANSFORMSOURCE_World, transformType=TRANSFORM_Local}
    B=ScalarParameter() {defaultValue=1, parameterName=MultWOP}
   ) {constB=5})
  Reroute(None=@MaterialExpressionConstantClamp_0)
  Transform(None=@n28)
  MaterialExpressionVertexColor_0 = ParticleColor()
 ]
}
