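# Translucent unlit Niagara ribbon material for the Musical Notes boost, drawing a rainbow spiral trail.
# Dynamic parameters pick the rainbow color and control desaturation, brightness and opacity.
# Layered sine and cosine waves build a spiral offset, with particle color channels repurposed for frequency, speed and amplitude.
# Static switches Stacked and Strings select textured dots, soft strings or procedural dots for opacity, faded along the ribbon and by particle alpha.
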Material /VehicleCosmetics/Boosts/Boost_MusicalNotes/Material/M_MusicalNotes_SpiralTrail
graph {
 properties {bUsedWithNiagaraRibbons=true, blendMode=BLEND_Translucent, pixelDepthOffsetMode=PDOM_Legacy, shadingModel=MSM_Unlit, usedShadingModels=Unlit}
 nodes [
  n1 = DynamicParameter() {paramNames [color, Color_Brightness, Desaturate, Opacity]}
  n2 = TextureSample(UVs=Add(A=TextureCoordinate() {vTiling=0}, B=@n1.color)) {texture=/DealershipSharedLibrary/FX/Textures/Diffuse/T_VCVFX_Rainbow_D}
  Desaturation(None=@n2.RGB, Fraction=@n1.Desaturate)
  MaterialFunctionCall(Emissive=Multiply(A=Desaturation(None=@n2.RGB, Fraction=@n1.Desaturate) {luminanceFactors=<rgba 0.3 0.59 0.11 0>}, B=@n1.Color_Brightness)) {materialFunction=&MF_FX_EmissiveAdjustments}
  n3 = ParticleColor()
  n4 = Multiply(A=@n3.G, B=OneMinus(None=Time()))
  n5 = ComponentMask(None=Multiply(A=@n3.R, B=TextureCoordinate() {coordinateIndex=1, uTiling=3})) {r=true}
  n6 = Add(A=@n4, B=@n5)
  n7 = Multiply(A=@n4, B=Constant() {r=0.5})
  n8 = Add(A=Add(A=Cosine(None=@n6), B=Cosine(None=@n6) {period=2}), B=Cosine(None=Add(A=@n5, B=@n7)) {period=0.75})
  n9 = Add(A=@n5, B=@n4)
  n10 = Sine(None=@n9)
  n11 = Sine(None=@n9) {period=2}
  n12 = AppendVector(A=@n8, B=Add(A=Add(A=@n10, B=@n11), B=Sine(None=@n9) {period=0.75}))
  n13 = Multiply(A=Multiply(A=AppendVector(A=@n12, B=@n8), B=@n3.B), B=@n3.A)
  n14 = Transform(None=@n13) {transformSourceType=TRANSFORMSOURCE_Local, transformType=TRANSFORM_Tangent}
  Multiply(A=CameraVectorWS(), B=@n14)
  n15 = ComponentMask(None=TextureCoordinate()) {g=true}
  n16 = Multiply(A=@n15, B=OneMinus(None=@n15))
  n17 = Constant()
  n18 = Saturate(None=Multiply(A=Subtract(A=@n16, B=@n17), B=Constant() {r=2}))
  n19 = OneMinus(None=ComponentMask(None=TextureCoordinate()) {r=true})
  n20 = Add(A=Frac(None=TextureCoordinate() {coordinateIndex=1, uTiling=24}), B=Constant() {r=-0.5})
  n21 = DotProduct(A=@n20, B=@n20)
  n22 = Multiply(A=Subtract(A=OneMinus(None=@n21), B=Constant() {r=0.95}), B=Constant() {r=12})
  n23 = StaticSwitchParameter(
   True=TextureSample(UVs=TextureCoordinate() {coordinateIndex=1, uTiling=24}) {texture=/VehicleCosmetics/Boosts/Boost_MusicalNotes/Textures/T_VCVFX_Boost_MusicNotes_Dots_A}.R
   False=StaticSwitchParameter(True=Multiply(A=@n19, B=@n18), False=Saturate(None=@n22)) {parameterName=Strings}
  ) {parameterName=Stacked}
  Multiply(A=ParticleColor().A, B=Multiply(A=@n1.Opacity, B=Multiply(A=@n19, B=@n23)))
  Multiply(A=@n14, B=ScalarParameter() {defaultValue=1, parameterName=MultWOP})
  Reroute(None=@n14)
  Transform(None=@n13)
  VectorParameter() {parameterName=WOPVAlue, defaultValue=<rgba 1.5 5 8 1>}
 ]
}
